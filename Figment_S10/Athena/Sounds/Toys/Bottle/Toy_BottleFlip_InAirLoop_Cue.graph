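# Sound cue for the looping in-air sound of the bottle flip toy, routed to the gadgets sound class at low volume.
# A random node picks one of three looping waves, a modulator varies its pitch slightly upward, and an enveloper ramps volume from half to full.
# Overridden attenuation uses logarithmic falloff, and overridden concurrency limits simultaneous instances per owner.
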SoundCue /Figment_S10/Athena/Sounds/Toys/Bottle/Toy_BottleFlip_InAirLoop_Cue
root=Enveloper_0
nodes [
 {
  id=Enveloper_0
  children [Modulator_0]
  properties {pitchCurve {editorCurveData {keys [{time=0}]}}, volumeCurve {editorCurveData {keys [<at 0 0.5>,<at 1 1>]}}}
 }
 {id=Modulator_0, children [Random_1], properties {pitchMax=1.2, pitchMin=1}}
 {id=Random_1, children [WavePlayer_0,WavePlayer_3,WavePlayer_4], properties {weights [1,1,1]}}
 {id=WavePlayer_0, properties {bLooping=true, soundWaveAssetPtr=&/ToyBottle_InAirLoop_03}}
 {id=WavePlayer_3, properties {bLooping=true, soundWaveAssetPtr=&/ToyBottle_InAirLoop_01}}
 {id=WavePlayer_4, properties {bLooping=true, soundWaveAssetPtr=&/ToyBottle_InAirLoop_02}}
]
properties {
 attenuationSettings=/Game/Sounds/Attenuation/Fort_ToyHit_Att
 bOverrideAttenuation=true
 bOverrideConcurrency=true
 maxDistance=2000
 soundClassObject=&Fort_Gadgets
 volumeMultiplier=0.1
 attenuationOverrides {distanceAlgorithm=Logarithmic, falloffDistance=2000, stereoSpread=512, attenuationShapeExtents {x=0}}
 concurrencyOverrides {bLimitToOwner=true, maxCount=2, volumeScale=0.7}
}
