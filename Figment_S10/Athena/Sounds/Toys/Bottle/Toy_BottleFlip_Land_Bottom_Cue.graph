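# Sound cue for the bottle flip toy landing on its bottom, in the gadgets sound class.
# It randomly picks one of three equally weighted landing waves, then passes it through a default Modulator for variation and an Enveloper whose volume rises from half to full.
# Cue-level overrides set logarithmic attenuation, in place of the assigned toy-hit asset, and cap concurrent plays per owner with volume scaling.
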SoundCue /Figment_S10/Athena/Sounds/Toys/Bottle/Toy_BottleFlip_Land_Bottom_Cue
root=Enveloper_0
nodes [
 {
  id=Enveloper_0
  children [Modulator_1]
  properties {pitchCurve {editorCurveData {keys [{time=0}]}}, volumeCurve {editorCurveData {keys [<at 0 0.5>,<at 1 1>]}}}
 }
 {id=Modulator_1, children [Random_0]}
 {id=Random_0, children [WavePlayer_1,WavePlayer_2,WavePlayer_5], properties {weights [1,1,1]}}
 {id=WavePlayer_1, properties {soundWaveAssetPtr=&/ToyBottle_Land_Bottom_01}}
 {id=WavePlayer_2, properties {soundWaveAssetPtr=&/ToyBottle_Land_Bottom_02}}
 {id=WavePlayer_5, properties {soundWaveAssetPtr=&/ToyBottle_Land_Bottom_03}}
]
properties {
 attenuationSettings=/Game/Sounds/Attenuation/Fort_ToyHit_Att
 bOverrideAttenuation=true
 bOverrideConcurrency=true
 maxDistance=2000
 soundClassObject=&Fort_Gadgets
 volumeMultiplier=0.6
 attenuationOverrides {distanceAlgorithm=Logarithmic, falloffDistance=2000, stereoSpread=512, attenuationShapeExtents {x=0}}
 concurrencyOverrides {bLimitToOwner=true, maxCount=2, volumeScale=0.7}
}
